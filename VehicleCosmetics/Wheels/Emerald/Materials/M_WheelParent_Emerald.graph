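# Parent material for Emerald vehicle cosmetic wheels, adding an animated rim glow to the attributes from MF_WheelParent_Simple.
# Two counter-panning noise samples are multiplied, added to a Fresnel term, and used to blend in the emissive glow, scaled by Glow Amount Max.
# The rim mask confines the glow and zeroes Metallic, Specular and Roughness there.
# The result feeds the shared vehicle damage and de-rez transition functions.
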Material /VehicleCosmetics/Wheels/Emerald/Materials/M_WheelParent_Emerald
graph {
 properties {
  bAutomaticallySetUsageInEditor=false
  bUseMaterialAttributes=true
  bUsedWithMeshDeformer=true
  bUsedWithSkeletalMesh=true
  pixelDepthOffsetMode=PDOM_Legacy
  usedShadingModels=Default Lit
 }
 nodes [
  n1 = NamedRerouteUsage()
  n2 = Constant() {r=1}
  n3 = ScalarParameter() {defaultValue=2, group=5_Special Effects, parameterName=Glow Amount Max}
  If(
   A=ComponentMask(None=Desaturation(None=@n1, Fraction=@n2) {luminanceFactors=<rgba 0.3 0.59 0.11 0>}) {r=true}
   "A > B"=@n3
   "A == B"=@n3
   "A < B"=Constant() {r=3}
  ) {constB=0.5}
  n4 = MaterialFunctionCall() {materialFunction=&MF_WheelParent_Simple}
  n5 = GetMaterialAttributes(None=@n4.MaterialAttributes) {attributeGetTypes [
    #1
    #2
    #3
    #4
    #5
    #6
   ]}
  n6 = ComponentMask(None=@n4.RimRGBAMask) {r=true}
  n7 = SetMaterialAttributes!(
   MaterialAttributes=@n5.MaterialAttributes
   Emissive Color=NamedRerouteUsage()
   Metallic=LinearInterpolate(A=@n5.Metallic, Alpha=@n6) {constB=0}
   Specular=LinearInterpolate(A=@n5.Specular, Alpha=@n6) {constB=0}
   Roughness=LinearInterpolate(A=@n5.Roughness, Alpha=@n6) {constB=0}
  ) {attributeSetTypes [
    #2
    #4
    #5
    #6
   ]}
  MaterialFunctionCall(In=MaterialFunctionCall(Material=@n7) {materialFunction=&MF_BaseVehicle_Damage}.Result) {materialFunction=&MF_VehicleDeRez_Transition}
  Multiply(B=@n1) {constA=0.5}
  NamedRerouteDeclaration(None=@n5.Base Color) {name=BC, nodeColor=<rgba 1 0 0.3984 1>}
  NamedRerouteDeclaration(None=@n4.VertexColorR) {name=VC_R, nodeColor=<rgba 1 0.375 0 1>}
  NamedRerouteDeclaration(None=@n5.Emissive Color) {name=Emissive, nodeColor=<rgba 0.2188 0 1 1>}
  n8 = Multiply(A=@n3, B=@n1)
  n9 = TextureSample(UVs=Panner(Coordinate=TextureCoordinate() {vTiling=0.5}) {speedX=0.1, speedY=-0.1}) {samplerSource=SSM_Wrap_WorldGroupSettings, texture=/VehicleCosmetics/Wheels/_SharedTextures/FX/ContrastNoiseGradient_T}
  n10 = TextureSample(UVs=Panner(Coordinate=TextureCoordinate() {uTiling=2, vTiling=0.5}) {speedX=-0.1, speedY=-0.11}) {samplerSource=SSM_Wrap_WorldGroupSettings, texture=&ContrastNoiseGradient_T}
  n11 = Add(
   A=Multiply(A=Multiply(A=@n9.RGB, B=@n10.RGB)) {constB=4}
   B=MaterialFunctionCall!(Normal=NamedRerouteUsage(), Exponent=ScalarParameter() {defaultValue=3, group=5_Special Effects, parameterName=Fresnel Amount}) {materialFunction=/VehicleCosmetics/Wheels/Cola_Holographic/Materials/MF_Fresnel_Classic}.Result
  )
  NamedRerouteDeclaration(None=Multiply(A=LinearInterpolate(B=@n8, Alpha=@n11), B=@n6)) {name=Emissive Final, nodeColor=<rgba 1 0.3516 0 1>}
  NamedRerouteDeclaration(None=@n5.Normal) {name=Normal, nodeColor=<rgba 0.5 0.5 1 1>}
 ]
}
